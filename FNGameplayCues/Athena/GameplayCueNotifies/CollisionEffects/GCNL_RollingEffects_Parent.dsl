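; Parent looping gameplay cue that plays rolling audio on physics objects, based on the PhysicsObjectRolling notify class.
; On application it attaches the layered audio component to the target's root, then sets and fades in the chosen sound if valid.
; Size-requirement checks on the target pick a small, medium or large sound variable, falling back to a default asset when that one is invalid.
; On removal it fades the audio out.
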(event EventOnApplication(EitherVFX-preferthisone)
  (params (MyTarget :type Actor%20Object%20Reference) (Parameters :type Gameplay%20Cue%20Parameters%20Structure%20%28by%20ref%29) (ParticleComponents :type Array%20of%20FXSystem%20Component%20Object%20References) (AudioComponents :type Array%20of%20Audio%20Component%20Object%20References) (BurstCameraShakeInstance :type Camera%20Shake%20Base%20Object%20Reference) (BurstDecalInstance :type Decal%20Actor%20Object%20Reference))
  (bind _asset (CallFunction|GetSoundAsset MyTarget))
  (bind _layeredaudiocomp (Variables|FortGameplayCueNotifyLoop_PhysicsObjectRolling|GetLayeredAudioComp))
  (Transformation|AttachComponentToComponent (Variables|FortGameplayCueNotifyLoop_PhysicsObjectRolling|GetLayeredAudioComp) (Variables|/Script/Engine.Actor|GetRootComponent MyTarget) "None" "SnapToTarget" "SnapToTarget" "SnapToTarget" true)
  (Utilities|IsValid _asset
    (:Is%20Valid
      (Audio|Components|Audio|SetSound _layeredaudiocomp _asset)
      (Audio|Components|Audio|FadeIn _layeredaudiocomp 0.25 1.0 0.0 "Linear" :!class /Script/Engine.AudioComponent))))

(event EventOnRemoval(EitherVFX-preferthisversion)
  (params (MyTarget :type Actor%20Object%20Reference) (Parameters :type Gameplay%20Cue%20Parameters%20Structure%20%28by%20ref%29) (ParticleComponents :type Array%20of%20FXSystem%20Component%20Object%20References) (AudioComponents :type Array%20of%20Audio%20Component%20Object%20References) (BurstCameraShakeInstance :type Camera%20Shake%20Base%20Object%20Reference) (BurstDecalInstance :type Decal%20Actor%20Object%20Reference))
  (Audio|Components|Audio|FadeOut (Variables|FortGameplayCueNotifyLoop_PhysicsObjectRolling|GetLayeredAudioComp) 1.0 0.0 "Linear" :!class /Script/Engine.AudioComponent))

(fn GetObjectSizeParam
  (params (PhysicsObject :type Actor%20Object%20Reference))
  (bind _returnvalue (FortPhysicsSimulationLibrary|DoesPhysicsObjectMeetSizeRequirement PhysicsObject "Small"))
  (if _returnvalue
    (return 0)
    (else
      (bind _returnvalue_2 (FortPhysicsSimulationLibrary|DoesPhysicsObjectMeetSizeRequirement PhysicsObject))
      (if _returnvalue_2
        (return 1)
        (else
          (return 2))))))

(fn GetSoundAsset
  (params (PhysicsObject :type Actor%20Object%20Reference))
  (return (select (Utilities|IsValid (Utilities|Select :Index (CallFunction|GetObjectSizeParam PhysicsObject) :Option%200 (Variables|Default|GetSoundAssetSmall) :Option%201 (Variables|Default|GetSoundAssetMedium) :Option%202 (Variables|Default|GetSoundAssetLarge)) :!class /Script/Engine.KismetSystemLibrary) (Utilities|Select :Index (CallFunction|GetObjectSizeParam PhysicsObject) :Option%200 (Variables|Default|GetSoundAssetSmall) :Option%201 (Variables|Default|GetSoundAssetMedium) :Option%202 (Variables|Default|GetSoundAssetLarge)) (Variables|Audio|GetSoundAssetDefault))))

(fn ConstructionScript)
