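; Burst gameplay cue notify for collision effects on physics boulder props, overriding the Niagara burst event.
; After calling the parent, it casts the target to the granite boulder, falling back to the desert boulder.
; For either type, it passes the boulder's root component velocity to the first spawned Niagara component as User.PhysicsVelocity.
; Desert boulders also set the User.DesertBoulder flag, letting one effect distinguish the variant.
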(fn OnBurst(Niagara)
  (params (MyTarget :type Actor%20Object%20Reference) (Parameters :type Gameplay%20Cue%20Parameters%20Structure%20%28by%20ref%29) (ParticleComponents :type Array%20of%20Niagara%20Particle%20System%20Component%20Object%20References) (AudioComponents :type Array%20of%20Audio%20Component%20Object%20References) (BurstCameraShakeInstance :type Camera%20Shake%20Base%20Object%20Reference) (BurstDecalInstance :type Decal%20Actor%20Object%20Reference))
  (bind _output (Utilities|Array|Get(acopy) ParticleComponents))
  (|CallParentFunction MyTarget Parameters ParticleComponents AudioComponents BurstCameraShakeInstance BurstDecalInstance)
  ;[1 1 1] GraniteBoulder
  (bind _asprop_physics_boulder_granite (Utilities|Casting|CastToProp_PhysicsBoulder_Granite MyTarget)
    (:then
      (Niagara|SetNiagaraVariable(Vector3) _output "User.PhysicsVelocity" (Transformation|GetComponentVelocity (Variables|/Script/Engine.Actor|GetRootComponent _asprop_physics_boulder_granite))))
    (:CastFailed
      ;[1 1 1] DesertBoulder
      (bind _asprop_physics_boulder_desert_01 (Utilities|Casting|CastToProp_PhysicsBoulder_Desert_01 MyTarget)
        (:then
          (Niagara|SetNiagaraVariable(Vector3) _output "User.PhysicsVelocity" (Transformation|GetComponentVelocity (Variables|/Script/Engine.Actor|GetRootComponent _asprop_physics_boulder_desert_01)))
          (Niagara|SetNiagaraVariable(Bool) _output "User.DesertBoulder" true))
        (:CastFailed
          _))
      (return))))
