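; Burst gameplay cue notify for water collision effects, named for the giant beach ball, that spawns a size-dependent splash.
; After calling the parent burst handler, it spawns a Cascade effect from the WaterSplash array at the cue location.
; GetObjectSizeParam picks the array index from the instigator's physics sim size: Medium and Large get their own entries, while all other sizes, including Giant, use the first.
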(fn GetObjectSizeParam
  (params (PhysicsObject :type Actor%20Object%20Reference))
  (bind _returnvalue (FortPhysicsSimulationLibrary|GetPhysicsObjectSimSize PhysicsObject))
  (return (Utilities|Select :Index _returnvalue :Small 0 :Average 0 :Medium 1 :Large 2 :Huge 0 :Giant 0 :Enormous 0 :Invalid 0)))

(fn OnBurst(Cascade)
  (params (MyTarget :type Actor%20Object%20Reference) (Parameters :type Gameplay%20Cue%20Parameters%20Structure%20%28by%20ref%29) (ParticleComponents :type Array%20of%20Cascade%20Particle%20System%20Component%20Object%20References) (AudioComponents :type Array%20of%20Audio%20Component%20Object%20References) (BurstCameraShakeInstance :type Camera%20Shake%20Base%20Object%20Reference) (BurstDecalInstance :type Decal%20Actor%20Object%20Reference))
  (bind (_normalizedmagnitude _rawmagnitude _effectcontext _matchedtagname _originaltag _aggregatedsourcetags _aggregatedtargettags _location _normal _instigator) (Ability|GameplayCue|BreakGameplayCueParameters Parameters))
  (|CallParentFunction MyTarget Parameters ParticleComponents AudioComponents BurstCameraShakeInstance BurstDecalInstance)
  (Effects|Components|ParticleSystem|SpawnFXAtLocation 0 (Utilities|Array|Get(acopy) (Variables|Default|GetWaterSplash) (CallFunction|GetObjectSizeParam _instigator)) _location)
  (return))
